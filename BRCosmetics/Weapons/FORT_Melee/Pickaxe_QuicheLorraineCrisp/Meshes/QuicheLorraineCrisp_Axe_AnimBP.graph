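# Animation Blueprint for the QuicheLorraineCrisp pickaxe cosmetic that adds physics-driven secondary motion to its two brush bones.
# The input pose passes through chained Anim Dynamics nodes for the right, then left brush bone, configured identically.
# Both simulate relative to the two-handed handle bone, using stiff angular springs, tight angular limits and a positive-Z gravity override in sim space.
# The event graph holds no logic.
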AnimBlueprint /BRCosmetics/Weapons/FORT_Melee/Pickaxe_QuicheLorraineCrisp/Meshes/QuicheLorraineCrisp_Axe_AnimBP
target_skeleton=&/QuicheLorraineCrisp_Axe_Skeleton
graphs [
 AnimGraph {nodes [
   AnimGraphNode_SubInput_0 = Input Pose()
   n1 = Anim Dynamics - Bone: dyn_Brush_r_01(ComponentPose=Local To Component(LocalPose=@AnimGraphNode_SubInput_0), Alpha=1, ExternalForce="0,0,0") {
    angularDampingOverride=0.9999
    angularSpringConstant=10000
    bAngularSpring=true
    bGravityOverrideInSimSpace=true
    bOverrideAngularDamping=true
    bUseGravityOverride=true
    numSolverIterationsPreUpdate=5
    simulationSpace=BoneRelative
    alphaBoolBlend {blendOption=Linear}
    boundBone {boneName=dyn_Brush_r_01}
    componentAppliedLinearAccClamp=<3000 3000 3000>
    componentLinearAccScale=<0.5 0.5 0.5>
    componentLinearVelScale=<0.5 0.5 0.5>
    gravityOverride {z=980}
    physicsBodyDefinitions [{
      boundBone {boneName=dyn_Brush_r_01}
      boxExtents=<25 25 25>
      constraintSetup {angularLimitsMax=<0 5 2>, angularLimitsMin=<0 -5 -2>, angularTarget=<1 0 0>}
      localJointOffset=<2 0 0>
     }]
    relativeSpaceBone {boneName=Melee_TwoHanded_Handle}
    simSpaceSettings {maxAngularAcceleration=10, maxAngularVelocity=2, simSpaceAngularAlpha=1, externalAngularVelocity {y=0.5}}
   }
   n2 = Anim Dynamics - Bone: dyn_Brush_l_01(ComponentPose=@n1, Alpha=1, ExternalForce="0,0,0") {
    angularDampingOverride=0.9999
    angularSpringConstant=10000
    bAngularSpring=true
    bGravityOverrideInSimSpace=true
    bOverrideAngularDamping=true
    bUseGravityOverride=true
    numSolverIterationsPreUpdate=5
    simulationSpace=BoneRelative
    alphaBoolBlend {blendOption=Linear}
    boundBone {boneName=dyn_Brush_l_01}
    componentAppliedLinearAccClamp=<3000 3000 3000>
    componentLinearAccScale=<0.5 0.5 0.5>
    componentLinearVelScale=<0.5 0.5 0.5>
    gravityOverride {z=980}
    physicsBodyDefinitions [{
      boundBone {boneName=dyn_Brush_l_01}
      boxExtents=<25 25 25>
      constraintSetup {angularLimitsMax=<0 5 2>, angularLimitsMin=<0 -5 -2>, angularTarget=<1 0 0>}
      localJointOffset=<2 0 0>
     }]
    relativeSpaceBone {boneName=Melee_TwoHanded_Handle}
    simSpaceSettings {maxAngularAcceleration=10, maxAngularVelocity=2, simSpaceAngularAlpha=1, externalAngularVelocity {y=0.5}}
   }
   Output Pose(Result=Component To Local(ComponentPose=@n2)) {layerGroup=None}
  ]}
 EventGraph {nodes [
   TryGetPawnOwner()
   Event BlueprintUpdateAnimation()
  ]}
]
variables []
